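# Niagara system for the OxideHoard emote: a red, rotating beacon-style light with a short flare and a burst of sparks.
# HeadLight keeps a spinning particle that feeds two mesh renderers with cone color and orientation attributes, plus a glow sprite and a light.
# The mesh-axis-to-camera dot product drives sprite size and brightness, combined with a flicker curve and System.Fade.
# Flare adds a quick stretching flash, while Sparks_CPU throws distance-culled, gravity-affected sparks.
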NiagaraSystem /BRCosmetics/Effects/Fort_Effects/Emotes/OxideHoard/NS_Emote_OxideHoard_Light
dependencies {
 usedDataInterfaces [
  &NiagaraDataInterfaceCamera
  &NiagaraDataInterfaceColorCurve
  &NiagaraDataInterfaceCurve
  &NiagaraDataInterfaceVector2DCurve
  &NiagaraDataInterfaceMeshRendererInfo
 ]
 usedDynamicInputs [&MakeVector4_FromLinearColor]
 usedModules [
  &SystemState
  &self:SystemSpawnScript.SystemScriptSource.SystemScriptGraph.NiagaraNodeAssignment_0.SetVariables_#1
 ]
}
emitters {
 Flare {
  emitterUpdateScript [
   &EmitterState(Life Cycle Mode=::NewEnumerator1, Loop Behavior=::NewEnumerator1, Loop Delay=0.4, Loop Duration=1, UseLoopDelay=false)
   &SpawnBurst_Instantaneous(Use Spawn Probability=false, UseLoopCountLimit=false)
  ]
  particleSpawnScript [
   &InitializeParticle(
    Color=<rgba 3 0.1 0.1 1>
    Lifetime=0.25
    Sprite Rotation Angle=-15
    Sprite Rotation Mode=::NewEnumerator0
    Sprite Size Mode=::NewEnumerator3
    Uniform Sprite Size=80
    UsePositionOffset=false
    Write Material Random=true
   )
   &CameraOffset(Camera Offset Amount=15)
  ]
  particleUpdateScript [
   &ParticleState()
   &ScaleSpriteSize(
    Non-Uniform Curve Sprite Scale=DataInterface() {propertyValues {
      XCurve {keys [
        {arriveTangent=2.0426, interpMode=RCIM_Cubic, leaveTangent=2.0426, tangentMode=RCTM_User, time=0.0756, value=0.1}
        {arriveTangent=-0.1514, interpMode=RCIM_Cubic, leaveTangent=-0.1514, tangentMode=RCTM_User, time=1, value=2}
       ]}
      YCurve {keys [
        {arriveTangent=-5.158, interpMode=RCIM_Cubic, leaveTangent=-5.158, tangentMode=RCTM_User, time=0.0126, value=1.7}
        {arriveTangent=-0.6848, interpMode=RCIM_Cubic, leaveTangent=-0.6848, tangentMode=RCTM_User, time=0.2302, value=0.8643}
        {arriveTangent=0.0392, interpMode=RCIM_Cubic, leaveTangent=0.0392, tangentMode=RCTM_User, time=1, value=0.1901}
       ]}
     }}
    Scale Sprite Size Mode=::NewEnumerator3
   )
   &ScaleColor(
    Scale Alpha=&FloatFromCurve(FloatCurve=DataInterface() {propertyValues {Curve {
        defaultValue=0
        keys [
         {arriveTangent=-2.673, interpMode=RCIM_Cubic, leaveTangent=-2.673, tangentMode=RCTM_User, time=0.1337, value=0.9847}
         {arriveTangent=-0.3973, interpMode=RCIM_Cubic, leaveTangent=-0.3973, tangentMode=RCTM_User, time=1, value=0}
        ]
       }}})
    Scale RGB=&VectorFromFloat(Value=1)
    ScaleA=true
    ScaleRGB=true
   )
   &SolveForcesAndVelocity(Clamp Velocity=false, Limit Acceleration=false)
  ]
  rendererClasses [&NiagaraSpriteRendererProperties]
  renderers [{rendererClass=&NiagaraSpriteRendererProperties}]
 }
 HeadLight {
  emitterUpdateScript [
   &EmitterState(
    Inactive Response=::NewEnumerator1
    Life Cycle Mode=::NewEnumerator1
    Loop Behavior=::NewEnumerator1
    Loop Delay=0.1
    Loop Duration Mode=::NewEnumerator1
    UseLoopDelay=false
   )
   &SpawnBurst_Instantaneous(Use Spawn Probability=false, UseLoopCountLimit=false)
  ]
  particleSpawnScript [
   &InitializeParticle(
    Lifetime=5
    Mesh Scale=<0.12 0.12 0.15>
    Mesh Scale Mode=::NewEnumerator4
    Position Offset=<0 0 2>
    Sprite Rotation Angle=-60
    Sprite Rotation Mode=::NewEnumerator0
    Sprite Size Mode=::NewEnumerator3
    Uniform Sprite Size=25
    UsePositionOffset=false
    Write Material Random=true
   )
   &InitialMeshOrientation(Mesh Axis To Orient=<0 0 -1>, Mesh Orientation Mode=::NewEnumerator2, Use Rotation Vector=false)
   &CameraOffset(Camera Offset Amount=5)
   /SetVariables(Particles.FlareOffset=10) {isSetParameters=true}
  ]
  particleUpdateScript [
   &ParticleState(Kill Particles When Lifetime Has Elapsed=false)
   &UpdateMeshOrientation(Rotation Coordinate Space=::NewEnumerator2, Rotation Vector=<0 0 1>)
   /SetVariables(Particles.LightDot:NiagaraFloat=&DotProduct(
     Vector A=&ReturnMeshOrientationAxis(Mesh Orientation Axis=::ZAxis, Mesh Orientation Quat=$Particles.MeshOrientation)
     Vector B=&ReturnCameraVector(Camera Vectors=::NewEnumerator1, Coordinate Space=::Local)
    )) {isSetParameters=true}
   /SetVariables(
    Particles.LightFacing:NiagaraFloat=&Abs_Float(Float:NiagaraFloat=$Particles.LightDot)
    Particles.LightSide:NiagaraFloat=&RemapRange(Input Value To Remap=$Particles.LightDot, Input Min=-1, Input Max=1, Output Min=0, Output Max=1)
   ) {isSetParameters=true}
   &ScaleColor(Color Scale:LinearColor=<rgba 0.8348 0.0083 0 1>, Scale Mode=3)
   /SetVariables(
    Particles.ConeColor:LinearColor=&Multiply_LinearColorByFloat(Linear Color=<rgba 0.8522 0.0085 0 1>, Float=$System.Fade)
    Particles.InnerFlareColor:LinearColor=&Multiply_LinearColorByFloat(
     Linear Color=$Particles.Color
     Float=&Multiply_Float(
      A=&FloatFromCurve(
       FloatCurve=DataInterface() {propertyValues {Curve {keys [
           <at 0 1>
           <at 0.0514 0.5534>
           <at 0.0629 1.6847>
           <at 0.0909 1>
           <at 0.2042 1>
           <at 0.2411 0.9235>
           <at 0.2749 1.6847>
           <at 0.3179 1>
           <at 0.407 1>
           <at 0.4531 1.6847>
           <at 0.4961 1>
           {arriveTangent=2.4659, leaveTangent=2.4659, time=0.6375, value=1}
           <at 0.6713 0.5534>
           <at 0.699 1.6847>
           <at 0.7389 1>
           <at 0.8587 1>
           <at 0.8987 1.6847>
           <at 0.9417 1>
           <at 1 1>
          ]}}}
       Scale Curve=&Multiply_Float(A=&RandomRangeFloat(Minimum=2, Maximum=3, Evaluation Type=::NewEnumerator1), B=$System.Fade)
      )
      B=&FloatFromCurve(
       FloatCurve=DataInterface() {propertyValues {Curve {keys [<at 0 0.5 RCIM_Cubic>, {arriveTangent=0.6896, interpMode=RCIM_Cubic, leaveTangent=0.6896, tangentMode=RCTM_User, time=1, value=1}]}}}
       CurveIndex=$Particles.LightFacing
      )
     )
    )
   ) {isSetParameters=true}
   &ScaleSpriteSize(
    Uniform Curve Index=$Particles.LightFacing
    Uniform Curve Sprite Scale=DataInterface() {propertyValues {Curve {
       defaultValue=0
       keys [
        {arriveTangent=0.1849, interpMode=RCIM_Cubic, leaveTangent=0.1849, tangentMode=RCTM_User, time=0, value=1}
        {
         arriveTangent=1.1641
         arriveTangentWeight=0.2704
         interpMode=RCIM_Cubic
         leaveTangent=1.1641
         leaveTangentWeight=0.063
         tangentMode=RCTM_User
         time=0.8111
         value=1.2945
        }
        {arriveTangent=10.9703, arriveTangentWeight=0.3333, interpMode=RCIM_Cubic, leaveTangent=10.9703, tangentMode=RCTM_User, time=1, value=3}
       ]
      }}}
   ) {module=ScaleSpriteSize001}
   &FN_Light_Attributes(Create Exponent Attribute=true, Create Radius Attribute=true, Radius=75)
   /SetVariables(Particles.SecondCone:Quat4f=&MakeQuaternion(
     Quaternion From=0
     Axis=&ReturnMeshOrientationAxis(Mesh Orientation Axis=::YAxis, Mesh Orientation Quat=$Particles.MeshOrientation)
     Angle=90
     From=?
     To=?
     Angle Type=::NewEnumerator0
     Coordinate Space=::Local
     Normalize Vectors=false
    )) {isSetParameters=true}
  ]
  rendererClasses [
   &NiagaraMeshRendererProperties
   &NiagaraSpriteRendererProperties
   &NiagaraLightRendererProperties
  ]
  renderers [
   {rendererClass=&NiagaraMeshRendererProperties, rendererIndex=0}
   {rendererClass=&NiagaraSpriteRendererProperties, rendererIndex=1}
   {rendererClass=&NiagaraLightRendererProperties, rendererIndex=2}
   {rendererClass=&NiagaraMeshRendererProperties, rendererIndex=3}
  ]
 }
 SmokeLWE {rendererClasses [], renderers []}
 Sparks_CPU {
  emitterUpdateScript [
   &EmitterState(
    Enable Distance Culling=true
    Life Cycle Mode=::NewEnumerator1
    Loop Behavior=::NewEnumerator1
    Loop Duration=0
    Max Distance Response=::KillImmediately
    Module.bApplySpawnCountScaleByDistance=true
    Module.bUseMaxDistance=true
    Module.bUseMinDistance=false
    Scalability Mode=::NewEnumerator1
    Scale Spawn Count=true
    Spawn Count Scale By Distance=DataInterface() {propertyValues {Curve {keys [<at 0.3676 0.9657>,<at 0.9962 0.4982>]}}}
    UseLoopDelay=false
   )
   &SpawnBurst_Instantaneous(Spawn Count=&UniformRangedInt(Minimum=3, Maximum=5), Use Spawn Probability=false)
  ]
  particleSpawnScript [
   &InitializeParticle(
    Color=<rgba 3 3 3 1>
    Lifetime Max=0.44
    Lifetime Min=0.23
    Lifetime Mode=::NewEnumerator1
    Sprite Size Mode=::NewEnumerator3
    Uniform Sprite Size=&Multiply_Float(
     A=&FN_FrontEndFloatModifier(Gameplay=1, Front End=1)
     B=&RandomRangeFloat(Minimum=2.5, Maximum=5)
    )
    UsePositionOffset=false
    Write Material Random=true
   )
   &ShapeLocation(
    Hemisphere Distribution=<1 0.5>
    MaskSpawnGroup=false
    Override Seed=false
    Sphere Radius=3
    Sphere Surface Distribution=1
    UseOwnerScale=true
   )
   &AddVelocity(
    Module.Use Falloff Distance=false
    Override Seed=false
    Velocity Mode=::NewEnumerator1
    Velocity Speed=&Multiply_Float(
     A=&RandomRangeFloat(Minimum=100, Maximum=250)
     B=&FN_FrontEndFloatModifier(Gameplay=1, Front End=1)
    )
   )
   &DynamicMaterialParameters(
    Index 0 Param 1=3
    Index 0 Param 2=20
    Index 0 Param 3=1
    Module.Param0WriteEnabled=true
    Module.Param1WriteEnabled=true
    Module.Param2WriteEnabled=true
    Module.Param3WriteEnabled=true
   )
  ]
  particleUpdateScript [
   &ParticleState()
   &GravityForce(Gravity=<0 0 -600>)
   &Drag(Drag=&UniformRangedFloat(Minimum=3, Maximum=8), Use Linear Drag=true, Use Rotational Drag=false)
   &SolveForcesAndVelocity(Clamp Velocity=false, Limit Acceleration=false)
   &ScaleSpriteSize(Scale Factor=&Vector2DFromCurve(
     Vector2Curve=DataInterface() {propertyValues {
       XCurve {keys [
         {arriveTangent=5.9728, interpMode=RCIM_Cubic, leaveTangent=5.9728, tangentMode=RCTM_User, time=0, value=0.1}
         {arriveTangent=-0.1, interpMode=RCIM_Cubic, leaveTangent=-0.1, tangentMode=RCTM_User, time=0.18, value=0.8}
         {interpMode=RCIM_Cubic, tangentMode=RCTM_User, time=1, value=0}
        ]}
       YCurve {keys [
         {arriveTangent=6.2715, interpMode=RCIM_Cubic, leaveTangent=6.2715, tangentMode=RCTM_User, time=0, value=0.1}
         {arriveTangent=-0.1001, interpMode=RCIM_Cubic, leaveTangent=-0.1001, tangentMode=RCTM_User, time=0.18, value=0.8}
         {interpMode=RCIM_Cubic, tangentMode=RCTM_User, time=0.9988, value=0}
        ]}
      }}
     ScaleCurve=<1 1>
    ))
   &ScaleSpriteSizeBySpeed(Max Scale Factor=<0.5 4>, Min Scale Factor=<0.5 1>, Velocity Threshold=666)
   &ScaleColor(
    Linear Color Curve=DataInterface() {propertyValues {
      AlphaCurve {keys [<at 0.5354 1>,<at 1 0>]}
      BlueCurve {keys [<at 0 0>,<at 0.1695 0.0172>,<at 1 0.06>]}
      GreenCurve {keys [<at 0 0.7969>,<at 0.1695 0.1694>,<at 1 0.02>]}
      RedCurve {keys [<at 0 4>,<at 0.1695 4.2834>,<at 1 4>]}
     }}
    Scale Mode=::NewEnumerator2
   )
  ]
  rendererClasses [&NiagaraSpriteRendererProperties]
  renderers [{rendererClass=&NiagaraSpriteRendererProperties}]
 }
}
summary {
 systemName=NS_Emote_OxideHoard_Light
 emitters [
  {
   emitterName=HeadLight
   rendererClasses [
    &NiagaraMeshRendererProperties
    &NiagaraSpriteRendererProperties
    &NiagaraLightRendererProperties
   ]
  }
  {emitterName=Flare, rendererClasses [&NiagaraSpriteRendererProperties]}
  {emitterName=Sparks_CPU, rendererClasses [&NiagaraSpriteRendererProperties]}
  {emitterName=SmokeLWE, rendererClasses []}
 ]
 userVariables [
  User.bIsFrontEnd {type=NiagaraBool}
  User.bIsFrontEndPreview {type=NiagaraBool}
 ]
}
