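# Niagara system for the OxideHoard emote's loudspeaker effect, emitting short-lived, faint sound-wave ripple meshes that travel along Z.
# LongRipples spawns small, rotating meshes that widen over life, with curve-animated material parameters.
# WideRipples spawns larger, uniformly growing meshes with constant material parameters.
# Both fade in quickly then out, scaled by user intensity values, system age and front-end alpha factors, and a time-of-day color tint.
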NiagaraSystem /BRCosmetics/Effects/Fort_Effects/Emotes/OxideHoard/NS_Emote_OxideHoard_LoudSpeaker
dependencies {
 usedDataInterfaces [&NiagaraDataInterfaceCurve]
 usedDynamicInputs [&FNTimeOfDayColor, &MakeCustomFloatFromBool]
 usedModules [
  &SystemState
  &self:SystemSpawnScript.SystemScriptSource.SystemScriptGraph.NiagaraNodeAssignment_0.SetVariables_#1
  &self:SystemSpawnScript.SystemScriptSource.SystemScriptGraph.NiagaraNodeAssignment_1.SetVariables_#2
  &self:SystemSpawnScript.SystemScriptSource.SystemScriptGraph.NiagaraNodeAssignment_2.SetVariables_#3
 ]
}
emitters {
 LongRipples {
  emitterUpdateScript [
   &EmitterState()
   &SpawnRate(SpawnRate=9, Use Spawn Probability=false)
  ]
  particleSpawnScript [
   &InitializeParticle(
    Color=<rgba 0.849 0.8806 1 0.015>
    Lifetime=&Multiply_Float(A=0.5, B=1)
    Mesh Scale=<0.15 0.15 0.5>
    Mesh Scale Mode=::NewEnumerator4
    UsePositionOffset=false
    Write Material Random=true
   )
   &AddVelocity(Velocity=&RandomRangeVector(Minimum=<0 0 140>, Maximum=<1 1 160>), Velocity Speed Scale=0.4)
   &InitialMeshOrientation(
    Mesh Orientation Mode=::NewEnumerator1
    Rotation=&RandomRangeVector(Minimum=<0 0 0>, Maximum=<0 0 1>)
    Use Rotation Vector=true
   )
   &SystemLocation(MaskSpawnGroup=false, Offset=<0 0 -7>)
  ]
  particleUpdateScript [
   &ParticleState()
   &UpdateMeshOrientation(Rotation Rate=0.4, Rotation Vector=<0 0 1>)
   &ScaleMeshSize(Scale Factor=&Multiply_VectorByFloat(
     Vector=<1 1 0.2>
     Float=&FloatFromCurve(
      FloatCurve=DataInterface() {propertyValues {Curve {keys [
          {interpMode=RCIM_Cubic, tangentMode=RCTM_User, time=0, value=0}
          {arriveTangent=0.7144, interpMode=RCIM_Cubic, leaveTangent=0.7144, tangentMode=RCTM_User, time=0.5198, value=0.1175}
          {arriveTangent=1.3089, interpMode=RCIM_Cubic, leaveTangent=1.3089, tangentMode=RCTM_User, time=1, value=1}
         ]}}}
      Scale Curve=4
     )
    ))
   &ScaleColor(
    Scale Alpha=&FloatFromCurve(
     FloatCurve=DataInterface() {propertyValues {Curve {keys [
         {arriveTangent=9.6668, interpMode=RCIM_Cubic, leaveTangent=9.6668, tangentMode=RCTM_User, time=0, value=0}
         <at 0.133 1 RCIM_Cubic>
         {interpMode=RCIM_Cubic, tangentMode=RCTM_User, time=1, value=0}
        ]}}}
     Scale Curve=&Multiply_Float(
      A=&Multiply_Float(A=$User.Intensity, B=$User.ColorIntensity)
      B=&Multiply_Float(A=$System.AgeFade, B=$System.FrontEnd_Alpha)
     )
    )
    Scale RGB=&Multiply_Vector(A=<1 1 1>, B=&MakeVectorFromLinearColorRGB(Color=$System.ToD))
    ScaleA=true
    ScaleRGB=true
   )
   &SolveForcesAndVelocity(Clamp Velocity=false, Limit Acceleration=false)
   &DynamicMaterialParameters(
    Index 0 Param 1=&FloatFromCurve(FloatCurve=DataInterface() {propertyValues {Curve {keys [<at 0.1 5 RCIM_Cubic>,<at 0.95 0 RCIM_Cubic>]}}})
    Index 0 Param 2=&FloatFromCurve(FloatCurve=DataInterface() {propertyValues {Curve {keys [<at 0 2.5 RCIM_Cubic>,<at 0.4 36 RCIM_Cubic>]}}})
    Index 0 Param 3=1
    Index 0 Param 4=&FloatFromCurve(FloatCurve=DataInterface() {propertyValues {Curve {keys [<at 0 1 RCIM_Cubic>,<at 1 0 RCIM_Cubic>]}}})
    Module.Param0WriteEnabled=true
    Module.Param1WriteEnabled=true
    Module.Param2WriteEnabled=true
    Module.Param3WriteEnabled=true
   ) {module=DynamicMaterialParameters002}
  ]
  rendererClasses [&NiagaraMeshRendererProperties]
  renderers [{rendererClass=&NiagaraMeshRendererProperties, rendererIndex=0}]
 }
 WideRipples {
  emitterUpdateScript [
   &EmitterState()
   &SpawnRate(SpawnRate=6.25, Use Spawn Probability=false)
  ]
  particleSpawnScript [
   &InitializeParticle(
    Color=<rgba 1 0.8787 0.8143 0.035>
    Lifetime=&Multiply_Float(A=0.3, B=1.2)
    Mesh Scale Mode=::NewEnumerator3
    Mesh Uniform Scale=70
    Sprite Rotation Mode=::NewEnumerator2
    UsePositionOffset=false
    Write Material Random=true
   )
   &AddVelocity(Velocity=&RandomRangeVector(Minimum=<0 0 140>, Maximum=<1 1 160>), Velocity Speed Scale=0.6)
  ]
  particleUpdateScript [
   &ParticleState()
   &UpdateMeshOrientation(Rotation Rate=0.2, Rotation Vector=<0 0 1>)
   &ScaleMeshSize(Scale Factor=&VectorFromFloat(Value=&FloatFromCurve(
      FloatCurve=DataInterface() {propertyValues {Curve {keys [{tangentMode=RCTM_User, time=0, value=0}, {tangentMode=RCTM_User, time=1, value=1}]}}}
      Scale Curve=1.6
     )))
   &ScaleColor(
    Scale Alpha=&FloatFromCurve(
     FloatCurve=DataInterface() {propertyValues {Curve {keys [
         {arriveTangent=9.6668, interpMode=RCIM_Cubic, leaveTangent=9.6668, tangentMode=RCTM_User, time=0, value=0}
         <at 0.133 1 RCIM_Cubic>
         {interpMode=RCIM_Cubic, tangentMode=RCTM_User, time=1, value=0}
        ]}}}
     Scale Curve=&Multiply_Float(
      A=&Multiply_Float(A=$User.Intensity, B=$User.ColorIntensity)
      B=&Multiply_Float(A=$System.AgeFade, B=$System.FrontEnd_Alpha)
     )
    )
    Scale RGB=&Multiply_Vector(A=<1 1 1>, B=&MakeVectorFromLinearColorRGB(Color=$System.ToD))
    ScaleA=true
    ScaleRGB=true
   )
   &SolveForcesAndVelocity(Clamp Velocity=false, Limit Acceleration=false)
   &DynamicMaterialParameters(
    Index 0 Param 1=0.5
    Index 0 Param 2=1
    Index 0 Param 3=1
    Index 0 Param 4=1
    Module.Param0WriteEnabled=true
    Module.Param1WriteEnabled=true
    Module.Param2WriteEnabled=true
    Module.Param3WriteEnabled=true
   )
  ]
  rendererClasses [&NiagaraMeshRendererProperties]
  renderers [{rendererClass=&NiagaraMeshRendererProperties, rendererIndex=0}]
 }
}
summary {
 systemName=NS_Emote_OxideHoard_LoudSpeaker
 emitters [
  {emitterName=LongRipples, rendererClasses [&NiagaraMeshRendererProperties]}
  {emitterName=WideRipples, rendererClasses [&NiagaraMeshRendererProperties]}
 ]
 userVariables [
  User.bIsFrontend {type=NiagaraBool}
  User.bIsFrontendPreview {type=NiagaraBool}
  User.ColorIntensity {type=NiagaraFloat, defaultValue=2, description=""}
  User.Intensity {type=NiagaraFloat, defaultValue=1, description=""}
 ]
}
